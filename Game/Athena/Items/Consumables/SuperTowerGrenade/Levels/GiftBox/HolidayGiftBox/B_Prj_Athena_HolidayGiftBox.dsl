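; Projectile Blueprint for the Holiday Gift Box consumable, filed under the SuperTowerGrenade gift box levels.
; It adds no logic of its own: OnExploded and the construction script only forward to the parent class.
; BeginPlay holds a disabled parent call, and the Tick and overlap events are empty stubs.
; Any gameplay behavior comes from the inherited class or from defaults outside this graph.
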(event EventBeginPlay
  (|CallParentFunction) ; This node is disabled and will not be called.\nDrag off pins to build functionality.
)

(event Collision|EventActorBeginOverlap
  (params (OtherActor :type Actor%20Object%20Reference)))

(event EventOnExploded
  (params (HitActors :type Array%20of%20Actor%20Object%20References) (HitResults :type Array%20of%20Hit%20Result%20Structures))
  (|CallParentFunction HitActors HitResults))

(event EventTick
  (params (DeltaSeconds :type Float%20%28single-precision%29)))

(fn ConstructionScript
  (|CallParentFunction))
